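# Physics asset for the F_MED_Rover_Pixel female character body cosmetic, covering its dangling accessories rather than a full-body ragdoll.
# Holster, pouch, strap and necklace each have a shaped body constrained with limited swing and twist to a tiny MSTR anchor body.
# The holster, pouch and strap anchors link to pelvis, the necklace's to spine_05, through angularly free constraints.
# Capsules on pelvis and spine_05 supply the torso shapes.
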PhysicsAsset /BRCosmetics/Characters/Player/Female/Medium/Bodies/F_MED_Rover_Pixel/Meshes/F_MED_Rover_Pixel_Physics
bodies {
 dyn_holster [{shapeType=Capsule, length=8.9704, radius=4.1804, shapeName=dyn_holster_capsule, center=<2.5172 0 4>, rotation=<rot 0 -9.9574 89.9999>}]
 dyn_holster_MSTR [{shapeType=Capsule, length=0.505, radius=0.505, shapeName=dyn_holster_MSTR_capsule, rotation=<rot 0 -90 0>}]
 dyn_necklace [{shapeType=Box, extentX=0.8095, extentY=4.389, extentZ=4.5981, shapeName=dyn_necklace_box, center=<2.8031 0 0>, rotation=<rot 0 90 0>}]
 dyn_necklace_MSTR [{shapeType=Capsule, length=0.505, radius=0.505, shapeName=dyn_necklace_MSTR_capsule, rotation=<rot 0 -90 0>}]
 dyn_pouch [{
   shapeType=Box
   extentX=11.3934
   extentY=5.2781
   extentZ=15.0287
   shapeName=dyn_pouch_box
   center=<7 0 5>
   rotation=<rot -0.0338 90.6359 -76.4944>
  }]
 dyn_pouch_MSTR [{shapeType=Capsule, length=0.505, radius=0.505, shapeName=dyn_pouch_MSTR_capsule, rotation=<rot 0 -90 0>}]
 dyn_strap [{shapeType=Capsule, length=3.2718, radius=1.7637, shapeName=dyn_strap_capsule, center=<3.8946 -1 0>, rotation=<rot 90 0 0>}]
 dyn_strap_MSTR [{shapeType=Capsule, length=0.505, radius=0.505, shapeName=dyn_strap_MSTR_capsule, rotation=<rot 0 -90 0>}]
 pelvis [
  {shapeType=Capsule, length=7.3208, radius=9.9686, center=<0 -1.1896 -0.0302>}
  {shapeType=Capsule, length=10.5625, radius=9.375, shapeName=pelvis_capsule, center=<-6.9251 -3.0356 -0.0001>}
 ]
 spine_05 [{shapeType=Capsule, length=9.7336, radius=9.251, shapeName=spine_05_capsule, center=<-4.0743 1.172 0>}]
}
constraints [
 {bone1Name=dyn_strap, bone2Name=dyn_strap_MSTR, swing1LimitDegrees=10, swing2LimitDegrees=15, twistLimitDegrees=3}
 {
  bone1Name=dyn_holster_MSTR
  bone2Name=pelvis
  swing1LimitDegrees=45
  swing1Motion=Free
  swing2LimitDegrees=45
  swing2Motion=Free
  twistLimitDegrees=45
  twistMotion=Free
 }
 {bone1Name=dyn_holster, bone2Name=dyn_holster_MSTR, swing1LimitDegrees=25, swing2LimitDegrees=15, twistLimitDegrees=3}
 {
  bone1Name=dyn_pouch_MSTR
  bone2Name=pelvis
  swing1LimitDegrees=45
  swing1Motion=Free
  swing2LimitDegrees=45
  swing2Motion=Free
  twistLimitDegrees=45
  twistMotion=Free
 }
 {bone1Name=dyn_pouch, bone2Name=dyn_pouch_MSTR, swing1LimitDegrees=10, swing2LimitDegrees=15, twistLimitDegrees=5}
 {
  bone1Name=dyn_strap_MSTR
  bone2Name=pelvis
  swing1LimitDegrees=45
  swing1Motion=Free
  swing2LimitDegrees=45
  swing2Motion=Free
  twistLimitDegrees=45
  twistMotion=Free
 }
 {
  bone1Name=dyn_necklace_MSTR
  bone2Name=spine_05
  swing1LimitDegrees=45
  swing1Motion=Free
  swing2LimitDegrees=45
  swing2Motion=Free
  twistLimitDegrees=45
  twistMotion=Free
 }
 {bone1Name=dyn_necklace, bone2Name=dyn_necklace_MSTR, swing1LimitDegrees=45, swing2LimitDegrees=45, twistLimitDegrees=15}
 {
  bone1Name=spine_05
  bone2Name=pelvis
  swing1LimitDegrees=45
  swing1Motion=Free
  swing2LimitDegrees=45
  swing2Motion=Free
  twistLimitDegrees=45
  twistMotion=Free
 }
]
