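# Post-process style AnimBP for the F_MED_Rover_Pixel body cosmetic, adding deformation and secondary motion to a pose copied from the parent mesh.
# Two Stone Viper Control Rigs run first, then additive bone edits on glute and shirt bones, driven by thigh-to-pelvis rotation alphas computed each update.
# Weighted constraints place strap, pouch, holster and necklace master bones before a pelvis-space RigidBody simulation.
# An AnimDynamics necklace node and second RigidBody sit unconnected.
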AnimBlueprint /BRCosmetics/Characters/Player/Female/Medium/Bodies/F_MED_Rover_Pixel/Meshes/F_MED_Rover_Pixel_AnimBP
target_skeleton=&/F_MED_Rover_Pixel_Skeleton
graphs [
 AnimGraph {nodes [
   Anim Dynamics - Bone: dyn_necklace(Alpha=1, ExternalForce="0,0,0") {
    angularDampingOverride=0.998
    angularSpringConstant=25
    bAngularSpring=true
    bGravityOverrideInSimSpace=true
    bOverrideAngularDamping=true
    bUseGravityOverride=true
    bUsePlanarLimit=false
    bUseSphericalLimits=true
    simulationSpace=RootRelative
    alphaBoolBlend {blendOption=Linear}
    boundBone {boneName=dyn_necklace}
    componentLinearAccScale=<0.3 0.3 0.6>
    componentLinearVelScale=<0.3 0.3 0.6>
    gravityOverride {z=-720}
    physicsBodyDefinitions [{
      collisionType=InnerSphere
      sphereCollisionRadius=1
      boundBone {boneName=dyn_necklace}
      boxExtents=<7 12 12>
      constraintSetup {twistAxis=AxisZ, angularLimitsMax=<15 0 0>, angularLimitsMin=<-15 0 -35>, angularTarget=<0 0 0>}
      localJointOffset=<3 0 0>
     }]
    relativeSpaceBone {boneName=spine_02}
   }
   RigidBody!(ComponentPose, bAlphaBoolEnabled:bool, OverrideWorldGravity:Vector="0,0,-980", ExternalForce) -> (Pose) {
    alphaInputType=Bool
    bClampLinearTranslationLimitToRefPose=true
    bOverrideWorldGravity=true
    overridePhysicsAsset=&/F_MED_Rover_Pixel_Physics
    simulationSpace=BaseBoneSpace
    alphaBoolBlend {blendInTime=0.1, blendOption=Linear, blendOutTime=0.1}
    baseBoneRef {boneName=pelvis}
    componentAppliedLinearAccClamp=<3000 3000 3000>
    componentLinearAccScale=<0.3 0.3 0.3>
    componentLinearVelScale=<0.3 0.3 0.3>
    overrideWorldGravity {z=-980}
    simSpaceSettings {
     maxAngularAcceleration=200
     maxAngularVelocity=5
     maxLinearAcceleration=1000
     maxLinearVelocity=300
     velocityScaleZ=0
     worldAlpha=1
     externalLinearDragV=<1 1>
    }
   }
   n1 = Control Rig(Source=Copy Pose From Mesh() {bUseAttachedParent=true}) {
    alphaBoolBlend {blendOption=Linear}
    controlRigAssetReference {blueprintRigClass=/BRCosmetics/Characters/Player/Female/Medium/Bodies/F_MED_Stone_Viper/Meshes/F_MED_Stone_Viper_Deform_CtrlRig_twists.F_MED_Stone_Viper_Deform_CtrlRig_twists_C}
   }
   n2 = Control Rig!(
    Source=@n1
    Alpha
    KneeRot_fr_l_114deg:Vector
    KneeTrans_fr_l_114deg:Vector="0.258,-3.591,0.009"
    KneeScale_fr_l_114deg:Vector="1,1,1"
    KneeScale_fr_l_130deg:Vector="1,1,1"
    KneeRot_fr_l_130deg:Vector="0,0,-51.2"
    KneeTrans_fr_l_130deg:Vector="0.258,-3.591,0.009"
   ) -> (Pose) {
    alphaBoolBlend {blendOption=Linear}
    controlRigAssetReference {blueprintRigClass=/BRCosmetics/Characters/Player/Female/Medium/Bodies/F_MED_Stone_Viper/Meshes/F_MED_Stone_Viper_Deform_CtrlRig_sculpters.F_MED_Stone_Viper_Deform_CtrlRig_sculpters_C}
   }
   n3 = Get thigh_rot_l()
   n4 = "Transform (Modify) Bone - Bone: dyn_glute_l"(
    ComponentPose=Local To Component(LocalPose=@n2)
    Alpha=@n3
    Translation="-4.689,5.014,-4.028"
    Rotation="0.000,0.000,-55"
    Scale="1.25,1.25,1.25"
   ) {
    rotationMode=BMM_Additive
    rotationSpace=BCS_BoneSpace
    scaleMode=BMM_Additive
    scaleSpace=BCS_BoneSpace
    translationMode=BMM_Additive
    translationSpace=BCS_BoneSpace
    alphaBoolBlend {blendOption=Linear}
    boneToModify {boneName=dyn_glute_l}
    rotation {roll=-55}
    scale=<1.25 1.25 1.25>
    translation=<-4.689 5.014 -4.028>
   }
   n5 = Get thigh_rot_r()
   n6 = "Transform (Modify) Bone - Bone: dyn_glute_r"(ComponentPose=@n4, Alpha=@n5, Translation="-3.75,0,0", Rotation="0.000,0.000,-55", Scale="1.3,1.3,1.3") {
    rotationMode=BMM_Additive
    rotationSpace=BCS_BoneSpace
    scaleMode=BMM_Additive
    scaleSpace=BCS_BoneSpace
    translationMode=BMM_Additive
    translationSpace=BCS_BoneSpace
    alphaBoolBlend {blendOption=Linear}
    boneToModify {boneName=dyn_glute_r}
    rotation {roll=-55}
    scale=<1.3 1.3 1.3>
    translation {x=-3.75}
   }
   n7 = "Transform (Modify) Bone - Bone: dyn_shirt_l"(ComponentPose=@n6, Alpha=@n3, Translation="-5,0,0", Rotation="0,0,0", Scale="1,1,1") {
    translationMode=BMM_Additive
    translationSpace=BCS_BoneSpace
    alphaBoolBlend {blendOption=Linear}
    boneToModify {boneName=dyn_shirt_l}
    translation {x=-5}
   }
   n8 = "Transform (Modify) Bone - Bone: dyn_shirt_r"(ComponentPose=@n7, Alpha=@n5, Translation="-5,0,0", Rotation="0,0,0", Scale="1,1,1") {
    translationMode=BMM_Additive
    translationSpace=BCS_BoneSpace
    alphaBoolBlend {blendOption=Linear}
    boneToModify {boneName=dyn_shirt_r}
    translation {x=-5}
   }
   n9 = "Constraint - dyn_strap_MSTR (PPPPPP)"!(
    ComponentPose=@n8
    Alpha=1
    ConstraintWeights_0=0.0153
    ConstraintWeights_1:float=0.0201
    ConstraintWeights_2:float=0.0025
    ConstraintWeights_3:float=0.0039
    ConstraintWeights_4:float=0.887
    ConstraintWeights_5:float=0.0712
   ) -> (Pose) {
    alphaBoolBlend {blendOption=Linear}
    boneToModify {boneName=dyn_strap_MSTR}
    constraintSetup [
     {targetBone {boneName=spine_01}}
     {targetBone {boneName=spine_02}}
     {targetBone {boneName=spine_04}}
     {targetBone {boneName=thigh_l}}
     {targetBone {boneName=pelvis}}
     {targetBone {boneName=thigh_twist_01_l}}
    ]
    constraintWeights [1,1,1,1,1,1]
   }
   n10 = "Constraint - dyn_pouch_MSTR (PPPP)"!(
    ComponentPose=@n9
    Alpha=1
    ConstraintWeights_0=0.0496
    ConstraintWeights_1:float=0.0438
    ConstraintWeights_2:float=0.9045
    ConstraintWeights_3:float=0.0021
   ) -> (Pose) {
    alphaBoolBlend {blendOption=Linear}
    boneToModify {boneName=dyn_pouch_MSTR}
    constraintSetup [{targetBone {boneName=spine_01}}, {targetBone {boneName=spine_02}}, {targetBone {boneName=pelvis}}, {targetBone {boneName=thigh_twist_01_l}}]
    constraintWeights [1,1,1,1]
   }
   n11 = "Constraint - dyn_holster_MSTR (PPP)"!(ComponentPose=@n10, Alpha=1, ConstraintWeights_0=0.0872, ConstraintWeights_1:float=0.0604, ConstraintWeights_2:float=0.8523) -> (Pose) {
    alphaBoolBlend {blendOption=Linear}
    boneToModify {boneName=dyn_holster_MSTR}
    constraintSetup [{targetBone {boneName=spine_01}}, {targetBone {boneName=spine_02}}, {targetBone {boneName=pelvis}}]
    constraintWeights [1,1,1]
   }
   n12 = "Constraint - dyn_necklace_MSTR (PPPPPPP)"!(
    ComponentPose=@n11
    Alpha=1
    ConstraintWeights_0=0
    ConstraintWeights_1:float=0.1607
    ConstraintWeights_2:float=0.8019
    ConstraintWeights_3:float=0.0222
    ConstraintWeights_4:float=0.0152
    ConstraintWeights_5:float=0
    ConstraintWeights_6:float=0
   ) -> (Pose) {
    alphaBoolBlend {blendOption=Linear}
    boneToModify {boneName=dyn_necklace_MSTR}
    constraintSetup [
     {targetBone {boneName=upperarm_twist_01_r}}
     {targetBone {boneName=spine_04}}
     {targetBone {boneName=spine_05}}
     {targetBone {boneName=clavicle_l}}
     {targetBone {boneName=clavicle_r}}
     {targetBone {boneName=deform_pectoral_fr_r}}
     {targetBone {boneName=deform_deltoid_fr_r}}
    ]
    constraintWeights [1,1,1,1,1,1,1]
   }
   n13 = RigidBody!(ComponentPose=@n12, Alpha:float, OverrideWorldGravity="0,0,-980", ExternalForce) -> (Pose) {
    bClampLinearTranslationLimitToRefPose=true
    bOverrideWorldGravity=true
    overridePhysicsAsset=&F_MED_Rover_Pixel_Physics
    simulationSpace=BaseBoneSpace
    alphaBoolBlend {blendOption=Linear}
    baseBoneRef {boneName=pelvis}
    componentAppliedLinearAccClamp=<3000 3000 3000>
    componentLinearAccScale=<0.3 0.3 0.5>
    componentLinearVelScale=<0.3 0.3 0.3>
   }
   Output Pose(Result=Component To Local(ComponentPose=@n13)) {layerGroup=None}
  ]}
 EventGraph {nodes [
   Event BlueprintUpdateAnimation()
   > Get!(execute:Exec) -> (then:Exec, else:Exec, Skel_Mesh:Skeletal Mesh Component Object Reference)
   > Sequence()
   n1 = Get Skel_Mesh()
   > Set thigh_rot_l(thigh_rot_l=GetAlphaFromReferenceTransformDelta(SkeletalMeshComponent=@n1, BoneName=thigh_l, BaseName=pelvis, InputRange=(X=-20,Y=-100), TwistAxis="1,0,0"))
   > Set thigh_rot_r(thigh_rot_r=GetAlphaFromReferenceTransformDelta(SkeletalMeshComponent=@n1, BoneName=thigh_r, BaseName=pelvis, InputRange=(X=-20,Y=-100), TwistAxis="1,0,0"))
   Event BlueprintInitializeAnimation()
   > Set Skel_Mesh(Skel_Mesh=GetOwnerSkeletalMeshComponent())
   TryGetPawnOwner()
  ]}
]
variables [
 Skel_Mesh {type=SkeletalMeshComponent}
 thigh_rot_l {type=number}
 thigh_rot_r {type=number}
]
