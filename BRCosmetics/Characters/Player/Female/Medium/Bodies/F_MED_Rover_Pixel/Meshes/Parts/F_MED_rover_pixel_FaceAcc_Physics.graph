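# Physics asset for the FaceAcc mesh part of the female medium Rover Pixel cosmetic body, covering a dynamic ponytail and the upper body around it.
# Capsule bodies cover the upper spine, clavicles, neck, head and ponytail bones.
# Body joints and the ponytail master and look-at links swing and twist freely.
# The five ponytail segments chain from the look-at bone with limited swing and narrower twist.
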PhysicsAsset /BRCosmetics/Characters/Player/Female/Medium/Bodies/F_MED_Rover_Pixel/Meshes/Parts/F_MED_rover_pixel_FaceAcc_Physics
bodies {
 clavicle_l [{shapeType=Capsule, length=12.6717, radius=4.8754, shapeName=clavicle_l_capsule, center=<5 0 0>, rotation=<rot 90 0 0>}]
 clavicle_r [{shapeType=Capsule, length=13.1634, radius=4.8754, shapeName=clavicle_r_capsule, center=<-5 0 0>, rotation=<rot 90 0 0>}]
 dyn_ponytail_01 [{shapeType=Capsule, length=1.4406, radius=2.5665, shapeName=dyn_ponytail_01_capsule, center=<2.5 0 0>}]
 dyn_ponytail_02 [{shapeType=Capsule, length=3.6013, radius=4.3927, shapeName=dyn_ponytail_02_capsule, center=<3.5 -1 0>}]
 dyn_ponytail_03 [{shapeType=Capsule, length=6.0699, radius=3.8622, shapeName=dyn_ponytail_03_capsule, center=<3.5 -0.7 0>}]
 dyn_ponytail_04 [{shapeType=Capsule, length=4.3258, radius=3.2891, shapeName=dyn_ponytail_04_capsule, center=<3.5 -0.5 0>}]
 dyn_ponytail_05 [{shapeType=Capsule, length=3.4062, radius=2.5092, shapeName=dyn_ponytail_05_capsule, center=<4 0 0>, rotation=<rot 90 0 0>}]
 dyn_ponytail_MSTR [{shapeType=Capsule, length=0.505, radius=0.505, shapeName=dyn_ponytail_MSTR_capsule, rotation=<rot 0 -90 0>}]
 dyn_ponytail_look_at [{shapeType=Capsule, length=0.505, radius=0.505, shapeName=dyn_ponytail_look_at_capsule, rotation=<rot 0 -90 0>}]
 head [{shapeType=Capsule, length=6.5934, radius=7.4787, shapeName=head_capsule, center=<0 2.1842 -0.0001>, rotation=<rot 90 0 0>}]
 neck_01 [{shapeType=Capsule, length=5.255, radius=5.005, shapeName=neck_01_capsule, center=<-1.102 0.6226 0>, rotation=<rot 0 -89.9999 0>}]
 spine_03 [{shapeType=Capsule, length=5.9414, radius=9.7882, shapeName=spine_03_capsule, center=<0 1.3061 0>, rotation=<rot 0 -89.9999 0>}]
 spine_04 [{shapeType=Capsule, length=7.0044, radius=10.6649, shapeName=spine_04_capsule, center=<2.5931 1.6603 0>, rotation=<rot 0 -92.6845 0>}]
 spine_05 [{shapeType=Capsule, length=16.7959, radius=8.2903, shapeName=spine_05_capsule, center=<1.675 -0.1132 -0.155>}]
}
constraints [
 {
  bone1Name=spine_05
  bone2Name=spine_04
  swing1LimitDegrees=45
  swing1Motion=Free
  swing2LimitDegrees=45
  swing2Motion=Free
  twistLimitDegrees=45
  twistMotion=Free
 }
 {
  bone1Name=clavicle_l
  bone2Name=spine_05
  swing1LimitDegrees=45
  swing1Motion=Free
  swing2LimitDegrees=45
  swing2Motion=Free
  twistLimitDegrees=45
  twistMotion=Free
 }
 {
  bone1Name=clavicle_r
  bone2Name=spine_05
  swing1LimitDegrees=45
  swing1Motion=Free
  swing2LimitDegrees=45
  swing2Motion=Free
  twistLimitDegrees=45
  twistMotion=Free
 }
 {
  bone1Name=spine_04
  bone2Name=spine_03
  swing1LimitDegrees=45
  swing1Motion=Free
  swing2LimitDegrees=45
  swing2Motion=Free
  twistLimitDegrees=45
  twistMotion=Free
 }
 {
  bone1Name=dyn_ponytail_look_at
  bone2Name=dyn_ponytail_MSTR
  swing1LimitDegrees=45
  swing1Motion=Free
  swing2LimitDegrees=45
  swing2Motion=Free
  twistLimitDegrees=45
  twistMotion=Free
 }
 {
  bone1Name=dyn_ponytail_MSTR
  bone2Name=head
  swing1LimitDegrees=45
  swing1Motion=Free
  swing2LimitDegrees=45
  swing2Motion=Free
  twistLimitDegrees=45
  twistMotion=Free
 }
 {
  bone1Name=head
  bone2Name=neck_01
  swing1LimitDegrees=45
  swing1Motion=Free
  swing2LimitDegrees=45
  swing2Motion=Free
  twistLimitDegrees=45
  twistMotion=Free
 }
 {
  bone1Name=neck_01
  bone2Name=spine_05
  swing1LimitDegrees=45
  swing1Motion=Free
  swing2LimitDegrees=45
  swing2Motion=Free
  twistLimitDegrees=45
  twistMotion=Free
 }
 {bone1Name=dyn_ponytail_01, bone2Name=dyn_ponytail_look_at, swing1LimitDegrees=45, swing2LimitDegrees=45, twistLimitDegrees=15}
 {bone1Name=dyn_ponytail_02, bone2Name=dyn_ponytail_01, swing1LimitDegrees=45, swing2LimitDegrees=45, twistLimitDegrees=15}
 {bone1Name=dyn_ponytail_03, bone2Name=dyn_ponytail_02, swing1LimitDegrees=45, swing2LimitDegrees=45, twistLimitDegrees=15}
 {bone1Name=dyn_ponytail_04, bone2Name=dyn_ponytail_03, swing1LimitDegrees=45, swing2LimitDegrees=45, twistLimitDegrees=15}
 {bone1Name=dyn_ponytail_05, bone2Name=dyn_ponytail_04, swing1LimitDegrees=45, swing2LimitDegrees=45, twistLimitDegrees=15}
]
