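# Post-process style AnimBP for the Rover Pixel female face accessory, animating its ponytail over a pose copied from the body.
# A cached pose with a weighted master-bone constraint feeds look-at plus RigidBody simulation, or AnimDynamics when back bling is detected outside the front end.
# Each update sets gravity, inertia scales, simulation-space state and left-arm alpha.
# The arm-driven corrective constraint chain is not connected to the output.
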AnimBlueprint /BRCosmetics/Characters/Player/Female/Medium/Bodies/F_MED_Rover_Pixel/Meshes/Parts/F_MED_Rover_Pixel_FaceAcc_AnimBP
target_skeleton=&/F_MED_Rover_Pixel_FaceAcc_Skeleton
graphs [
 AnimGraph {nodes [
   n1 = "Constraint - dyn_ponytail_01_con (PP)"!(
    ComponentPose=Local To Component(LocalPose=Use cached pose 'prerigid'())
    Alpha=1
    ConstraintWeights_0=0.4526
    ConstraintWeights_1:float=0.5474
   ) -> (Pose) {
    alphaBoolBlend {blendOption=Linear}
    boneToModify {boneName=dyn_ponytail_01_con}
    constraintSetup [{targetBone {boneName=head}}, {targetBone {boneName=neck_02}}]
    constraintWeights [1,1]
   }
   n2 = "Constraint - dyn_ponytail_02_con (PPP)"!(ComponentPose=@n1, Alpha=1, ConstraintWeights_0=0.2708, ConstraintWeights_1:float=0.1078, ConstraintWeights_2:float=0.6214) -> (Pose) {
    alphaBoolBlend {blendOption=Linear}
    boneToModify {boneName=dyn_ponytail_02_con}
    constraintSetup [{targetBone {boneName=head}}, {targetBone {boneName=neck_02}}, {targetBone {boneName=neck_01}}]
    constraintWeights [1,1,1]
   }
   n3 = "Constraint - dyn_ponytail_03_con (PP)"!(ComponentPose=@n2, Alpha=1, ConstraintWeights_0=0.23, ConstraintWeights_1:float=0.77) -> (Pose) {
    alphaBoolBlend {blendOption=Linear}
    boneToModify {boneName=dyn_ponytail_03_con}
    constraintSetup [{targetBone {boneName=spine_05}}, {targetBone {boneName=clavicle_l}}]
    constraintWeights [1,1]
   }
   n4 = "Constraint - dyn_ponytail_04_con (PP)"!(ComponentPose=@n3, Alpha=1, ConstraintWeights_0=0.43, ConstraintWeights_1:float=0.57) -> (Pose) {
    alphaBoolBlend {blendOption=Linear}
    boneToModify {boneName=dyn_ponytail_04_con}
    constraintSetup [{targetBone {boneName=spine_05}}, {targetBone {boneName=clavicle_l}}]
    constraintWeights [1,1]
   }
   n5 = "Constraint - dyn_ponytail_05_con (PPP)"!(ComponentPose=@n4, Alpha=1, ConstraintWeights_0=0.6591, ConstraintWeights_1:float=0.2295, ConstraintWeights_2:float=0.1114) -> (Pose) {
    alphaBoolBlend {blendOption=Linear}
    boneToModify {boneName=dyn_ponytail_05_con}
    constraintSetup [{targetBone {boneName=spine_05}}, {targetBone {boneName=spine_04}}, {targetBone {boneName=clavicle_l}}]
    constraintWeights [1,1,1]
   }
   n6 = Get Arm Rot() -> (Arm Rot:double)
   n7 = "Transform (Modify) Bone - Bone: dyn_ponytail_03_con"(ComponentPose=@n5, Alpha=@n6, Translation="0.024,-3.303,-1.533", Rotation="0,0,0", Scale="1,1,1") {
    scaleMode=BMM_Additive
    scaleSpace=BCS_BoneSpace
    translationMode=BMM_Additive
    translationSpace=BCS_BoneSpace
    alphaBoolBlend {blendOption=Linear}
    boneToModify {boneName=dyn_ponytail_03_con}
    translation=<0.024 -3.303 -1.533>
   }
   n8 = "Transform (Modify) Bone - Bone: dyn_ponytail_04_con"(ComponentPose=@n7, Alpha=@n6, Translation="1.939,-0.294,2.168", Rotation="0,0,0", Scale="1,1,1") {
    scaleMode=BMM_Additive
    scaleSpace=BCS_BoneSpace
    translationMode=BMM_Additive
    translationSpace=BCS_BoneSpace
    alphaBoolBlend {blendOption=Linear}
    boneToModify {boneName=dyn_ponytail_04_con}
    translation=<1.939 -0.294 2.168>
   }
   n9 = "Constraint - dyn_ponytail_01 (P)"(ComponentPose=@n8, Alpha=1, ConstraintWeights_0=1) {
    alphaBoolBlend {blendOption=Linear}
    boneToModify {boneName=dyn_ponytail_01}
    constraintSetup [{offsetOption=None, targetBone {boneName=dyn_ponytail_01_con}}]
    constraintWeights [1]
   }
   n10 = "Constraint - dyn_ponytail_02 (P)"(ComponentPose=@n9, Alpha=1, ConstraintWeights_0=1) {
    alphaBoolBlend {blendOption=Linear}
    boneToModify {boneName=dyn_ponytail_02}
    constraintSetup [{offsetOption=None, targetBone {boneName=dyn_ponytail_02_con}}]
    constraintWeights [1]
   }
   n11 = "Constraint - dyn_ponytail_03 (P)"(ComponentPose=@n10, Alpha=1, ConstraintWeights_0=1) {
    alphaBoolBlend {blendOption=Linear}
    boneToModify {boneName=dyn_ponytail_03}
    constraintSetup [{offsetOption=None, targetBone {boneName=dyn_ponytail_03_con}}]
    constraintWeights [1]
   }
   n12 = "Constraint - dyn_ponytail_04 (P)"(ComponentPose=@n11, Alpha=1, ConstraintWeights_0=1) {
    alphaBoolBlend {blendOption=Linear}
    boneToModify {boneName=dyn_ponytail_04}
    constraintSetup [{offsetOption=None, targetBone {boneName=dyn_ponytail_04_con}}]
    constraintWeights [1]
   }
   n13 = "Constraint - dyn_ponytail_05 (P)"(ComponentPose=@n12, Alpha=1, ConstraintWeights_0=1) {
    alphaBoolBlend {blendOption=Linear}
    boneToModify {boneName=dyn_ponytail_05}
    constraintSetup [{offsetOption=None, targetBone {boneName=dyn_ponytail_05_con}}]
    constraintWeights [1]
   }
   n14 = Get Arm Rot()
   n15 = "Transform (Modify) Bone - Bone: dyn_ponytail_03"(ComponentPose=@n13, Alpha=@n14, Translation="0,0,0", Rotation="0,0,0", Scale="1.000,1.000,0.750") {
    scaleMode=BMM_Additive
    scaleSpace=BCS_BoneSpace
    translationMode=BMM_Additive
    translationSpace=BCS_BoneSpace
    alphaBoolBlend {blendOption=Linear}
    boneToModify {boneName=dyn_ponytail_03}
    scale {z=0.75}
   }
   n16 = "Transform (Modify) Bone - Bone: dyn_ponytail_04"(ComponentPose=@n15, Alpha=@n14, Translation="0,0,0", Rotation="0,0,0", Scale="1.000,1.000,0.750") {
    scaleMode=BMM_Additive
    scaleSpace=BCS_BoneSpace
    translationMode=BMM_Additive
    translationSpace=BCS_BoneSpace
    alphaBoolBlend {blendOption=Linear}
    boneToModify {boneName=dyn_ponytail_04}
    scale {z=0.75}
   }
   n17 = Anim Dynamics - Bone: dyn_ponytail_05(ComponentPose=@n16, Alpha=0.4, ExternalForce="0,0,0") {
    angularDampingOverride=0.9999
    angularSpringConstant=75
    bAngularSpring=true
    bGravityOverrideInSimSpace=true
    bOverrideAngularDamping=true
    simulationSpace=RootRelative
    alphaBoolBlend {blendOption=Linear}
    boundBone {boneName=dyn_ponytail_05}
    componentAppliedLinearAccClamp=<3500 3500 3500>
    componentLinearAccScale=<0.4 0.4 0.4>
    componentLinearVelScale=<0.4 0.4 0.4>
    physicsBodyDefinitions [{
      sphereCollisionRadius=0
      boundBone {boneName=dyn_ponytail_05}
      boxExtents=<15 10 10>
      constraintSetup {angularLimitsMax=<1 15 0>, angularLimitsMin=<-1 -15 -15>, angularTarget=<0 0 0>}
      localJointOffset=<2 0 0>
     }]
    relativeSpaceBone {boneName=root}
   }
   Component To Local(ComponentPose=@n17)
   n18 = Look At - Bone: dyn_ponytail_look_at(ComponentPose=Local To Component(LocalPose=Use cached pose 'prerigid'()), Alpha=1) {
    alphaBoolBlend {blendOption=Linear}
    boneToModify {boneName=dyn_ponytail_look_at}
    lookAtLocation=<0 -12>
    lookAtTarget {boneReference {boneName=pelvis}}
    lookAt_Axis {axis=<1 0>}
   }
   n19 = "Constraint - dyn_ponytail_01 (PP)"!(ComponentPose=@n18, bAlphaBoolEnabled:bool=True, ConstraintWeights_0=1, ConstraintWeights_1:float=0) -> (Pose) {
    alphaInputType=Bool
    bAlphaBoolEnabled=false
    alphaBoolBlend {blendInTime=0.1, blendOption=Linear, blendOutTime=0.1}
    boneToModify {boneName=dyn_ponytail_01}
    constraintSetup [{targetBone {boneName=spine_05}}, {targetBone {boneName=head}}]
    constraintWeights [1,1]
   }
   n20 = Anim Dynamics - Bone: dyn_ponytail_03(ComponentPose=@n19, Alpha=0.4, ExternalForce="0,0,0") {
    angularDampingOverride=0.9999
    angularSpringConstant=75
    bAngularSpring=true
    bGravityOverrideInSimSpace=true
    bOverrideAngularDamping=true
    simulationSpace=RootRelative
    alphaBoolBlend {blendOption=Linear}
    boundBone {boneName=dyn_ponytail_03}
    componentAppliedLinearAccClamp=<3500 3500 3500>
    componentLinearAccScale=<0.4 0.4 0.4>
    componentLinearVelScale=<0.4 0.4 0.4>
    physicsBodyDefinitions [{
      sphereCollisionRadius=0
      boundBone {boneName=dyn_ponytail_03}
      boxExtents=<15 10 10>
      constraintSetup {angularLimitsMax=<1 15 0>, angularLimitsMin=<-1 -15 -15>, angularTarget=<0 0 0>}
      localJointOffset=<2 0 0>
     }]
    relativeSpaceBone {boneName=root}
   }
   n21 = Anim Dynamics - Bone: dyn_ponytail_04(ComponentPose=@n20, Alpha=0.4, ExternalForce="0,0,0") {
    angularDampingOverride=0.9999
    angularSpringConstant=75
    bAngularSpring=true
    bGravityOverrideInSimSpace=true
    bOverrideAngularDamping=true
    simulationSpace=RootRelative
    alphaBoolBlend {blendOption=Linear}
    boundBone {boneName=dyn_ponytail_04}
    componentAppliedLinearAccClamp=<3500 3500 3500>
    componentLinearAccScale=<0.4 0.4 0.4>
    componentLinearVelScale=<0.4 0.4 0.4>
    physicsBodyDefinitions [{
      sphereCollisionRadius=0
      boundBone {boneName=dyn_ponytail_04}
      boxExtents=<15 10 10>
      constraintSetup {angularLimitsMax=<1 15 0>, angularLimitsMin=<-1 -15 -15>, angularTarget=<0 0 0>}
      localJointOffset=<2 0 0>
     }]
    relativeSpaceBone {boneName=root}
   }
   n22 = Anim Dynamics - Bone: dyn_ponytail_05(ComponentPose=@n21, Alpha=0.4, ExternalForce="0,0,0") {
    angularDampingOverride=0.9999
    angularSpringConstant=75
    bAngularSpring=true
    bGravityOverrideInSimSpace=true
    bOverrideAngularDamping=true
    simulationSpace=RootRelative
    alphaBoolBlend {blendOption=Linear}
    boundBone {boneName=dyn_ponytail_05}
    componentAppliedLinearAccClamp=<3500 3500 3500>
    componentLinearAccScale=<0.4 0.4 0.4>
    componentLinearVelScale=<0.4 0.4 0.4>
    physicsBodyDefinitions [{
      sphereCollisionRadius=0
      boundBone {boneName=dyn_ponytail_05}
      boxExtents=<15 10 10>
      constraintSetup {angularLimitsMax=<1 15 0>, angularLimitsMin=<-1 -15 -15>, angularTarget=<0 0 0>}
      localJointOffset=<2 0 0>
     }]
    relativeSpaceBone {boneName=root}
   }
   n23 = Look At - Bone: dyn_ponytail_look_at(ComponentPose=Local To Component(LocalPose=Use cached pose 'prerigid'()), Alpha=1) {
    alphaBoolBlend {blendOption=Linear}
    boneToModify {boneName=dyn_ponytail_look_at}
    lookAtLocation {x=0}
    lookAtTarget {boneReference {boneName=dyn_root_lookat}}
    lookAt_Axis {axis=<1 0>}
   }
   n24 = RigidBody!(
    ComponentPose=@n23
    Alpha
    OverrideWorldGravity:Vector=Get GravityOverride()
    ExternalForce
    ComponentLinearAccScale:Vector=Get AccScale() -> (AccScale:Vector)
    ComponentLinearVelScale:Vector=Get VelScale() -> (VelScale:Vector)
   ) -> (Pose) {
    bClampLinearTranslationLimitToRefPose=true
    bOverrideWorldGravity=true
    overridePhysicsAsset=&/F_MED_rover_pixel_FaceAcc_Physics
    simulationSpace=BaseBoneSpace
    alphaBoolBlend {blendOption=Linear}
    baseBoneRef {boneName=spine_05}
    componentAppliedLinearAccClamp=<3000 3000 3000>
    componentLinearAccScale=<0.4 0.4 0.6>
    componentLinearVelScale=<0.4 0.4 0.6>
   }
   n25 = Blend Poses by bool(
    BlendPose_0=Component To Local(ComponentPose=@n22)
    BlendPose_1=Component To Local(ComponentPose=@n24)
    bActiveValue=Get backpack() -> (backpack:bool)
   ) {blendType=Linear}
   n26 = Look At - Bone: dyn_ponytail_look_at(ComponentPose=Local To Component(LocalPose=Use cached pose 'prerigid'()), Alpha=1) {
    alphaBoolBlend {blendOption=Linear}
    boneToModify {boneName=dyn_ponytail_look_at}
    lookAtLocation {x=0}
    lookAtTarget {boneReference {boneName=dyn_root_lookat}}
    lookAt_Axis {axis=<1 0>}
   }
   n27 = RigidBody(
    ComponentPose=@n26
    OverrideWorldGravity=Get GravityOverride()
    ComponentLinearAccScale=Get AccScale()
    ComponentLinearVelScale=Get VelScale()
   ) {
    bClampLinearTranslationLimitToRefPose=true
    bOverrideWorldGravity=true
    overridePhysicsAsset=&F_MED_rover_pixel_FaceAcc_Physics
    simulationSpace=BaseBoneSpace
    alphaBoolBlend {blendOption=Linear}
    baseBoneRef {boneName=spine_05}
    componentAppliedLinearAccClamp=<3000 3000 3000>
    componentLinearAccScale=<0.4 0.4 0.6>
    componentLinearVelScale=<0.4 0.4 0.6>
   }
   n28 = Component To Local(ComponentPose=@n27)
   n29 = Get bIsInFrontEnd()
   n30 = "Transform (Modify) Bone - Bone: dyn_ponytail_01"(
    ComponentPose=Local To Component(LocalPose=Blend Poses by bool(BlendPose_0=@n28, BlendPose_1=@n25, bActiveValue=@n29) {blendType=Linear})
    Alpha=1
    Translation="0,0,0"
    Rotation="0,0,0"
    Scale="1,1,1"
   ) {
    scaleMode=BMM_Additive
    scaleSpace=BCS_BoneSpace
    translationMode=BMM_Additive
    translationSpace=BCS_BoneSpace
    alphaBoolBlend {blendOption=Linear}
    boneToModify {boneName=dyn_ponytail_01}
   }
   Output Pose(Result=Component To Local(ComponentPose=@n30)) {layerGroup=None}
   n31 = Copy Pose From Mesh(SourceMeshComponent=Get Body_Mesh())
   n32 = Copy Pose From Mesh() {bCopyCurves=true, bUseAttachedParent=true}
   n33 = "Constraint - dyn_ponytail_MSTR (PPPPP)"!(
    ComponentPose=Local To Component(LocalPose=Layered blend per bone(BasePose=@n32, BlendPoses_0=@n31) {layerSetup [{branchFilters [{boneName=upperarm_l}, {boneName=upperarm_r}]}]})
    Alpha=1
    ConstraintWeights_0=0.0178
    ConstraintWeights_1:float=0.1266
    ConstraintWeights_2:float=0.175
    ConstraintWeights_3:float=0.3477
    ConstraintWeights_4:float=0.333
   ) -> (Pose) {
    alphaBoolBlend {blendOption=Linear}
    boneToModify {boneName=dyn_ponytail_MSTR}
    constraintSetup [
     {targetBone {boneName=spine_03}}
     {targetBone {boneName=spine_05}}
     {targetBone {boneName=neck_01}}
     {targetBone {boneName=head}}
     {targetBone {boneName=neck_02}}
    ]
    constraintWeights [1,1,1,1,1]
   }
   Save cached pose 'prerigid'(Pose=Component To Local(ComponentPose=@n33)) {cacheName=prerigid}
   Get bRigidBodyEnabled()
  ]}
 EventGraph {nodes [
   Event BlueprintUpdateAnimation()
   > n1 = Sequence!(execute) -> (then_0, then_1, then_2:Exec, then_3:Exec)
   > n2 = Gravity Override State Function DEPRECATED(CharacterPartAnimInstance=Self-Reference(), GameTimeInSeconds=GetGameTimeInSeconds(), GravityOverrideParams=Get Gravity Override Params())
   > Set GravityOverride(GravityOverride=@n2)
   n3 = ComplexSimSpaceFunction DEPRECATED!(
    execute=@n1.then_1
    CharacterPartAnimInstance:Custom Character Part Anim Instance Object Reference=Self-Reference()
    InterpSpeedMax:double=9
    PawnSpeedMax:double=20
    RInterpErrorTolerance:double=1
    DeltaTime:double=0
    DebugDraw:bool=false
    sim_space_rotation_in:Rotator=Get sim_space_rotation() -> (sim_space_rotation:Rotator)
    inComponentSpaceIn:bool=Get InComponentSpace() -> (InComponentSpace:bool)
    inComponentSpaceLastFrameIn:bool=Get InComponentSpaceLastFrame() -> (InComponentSpaceLastFrame:bool)
    lockedOnSimSpaceIn:bool=Get LockedOnSimSpace() -> (LockedOnSimSpace:bool)
    InterpSpeedIn:double=Get InterpSpeed() -> (InterpSpeed:double)
   ) -> (
    then
    sim_space_rotation:Rotator
    inComponentSpaceOut:bool
    inComponentSpaceLastFrameOut:bool
    lockedOnSimSpaceOut:bool
    InterpSpeedOut:double
   )
   > Set sim_space_rotation(sim_space_rotation:Rotator=@n3) -> (Output_Get:Rotator)
   > Set InComponentSpace(InComponentSpace:bool=@n3.inComponentSpaceOut) -> (Output_Get:bool)
   > Set InComponentSpaceLastFrame(InComponentSpaceLastFrame:bool=@n3.inComponentSpaceLastFrameOut) -> (Output_Get:bool)
   > Set LockedOnSimSpace(LockedOnSimSpace:bool=@n3.lockedOnSimSpaceOut) -> (Output_Get:bool)
   > Set InterpSpeed(InterpSpeed:double=@n3.InterpSpeedOut) -> (Output_Get:double)
   n4 = Branch(execute=@n1.then_2, Condition=CheckForBackbling())
   > Set AccScale(AccScale:Vector="0, 0, 0") -> (Output_Get:Vector)
   > Set VelScale(VelScale:Vector="0, 0, 0") -> (Output_Get:Vector)
   > Set GravityOverride(GravityOverride="0, 0,-980")
   > Set backpack(backpack:bool=true) -> (Output_Get:bool)
   Set AccScale(execute=@n4.else, AccScale="0.4,0.4,0.6")
   > Set VelScale(VelScale="0.4,0.4,0.6")
   > Set backpack(backpack=false)
   Set Arm Rot(
    execute=@n1.then_3
    Arm Rot:double=GetAlphaFromReferenceTransformDelta(SkeletalMeshComponent=Get Skel_Mesh(), BoneName=upperarm_l, BaseName=spine_05, InputRange=(X=60,Y=100), TwistAxis="1,0,0")
   ) -> (Output_Get:double)
   Event BlueprintInitializeAnimation()
   n5 = GetOwnerSkeletalMeshComponent()
   > Set Skel_Mesh(Skel_Mesh=@n5)
   > Set Body_Mesh(Body_Mesh=@n5)
   TryGetPawnOwner()
  ]}
]
variables [
 AccScale {type=Vector}
 Arm Rot {type=number}
 backpack {type=boolean}
 Body_Mesh {type=SkeletalMeshComponent}
 Gravity Override Params {type=GravityOverrideParamsStruct}
 GravityOverride {type=Vector}
 InComponentSpace {type=boolean}
 InComponentSpaceLastFrame {type=boolean}
 InterpSpeed {type=number}
 LockedOnSimSpace {type=boolean}
 sim_space_rotation {type=Rotator}
 Skel_Mesh {type=SkeletalMeshComponent}
 VelScale {type=Vector}
]
